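# EQS query that picks a goal for terrestrial wildlife AI pawns from perceived enemies and loot.
# Filters keep loot that matches the forage tag query, is within range and near navmesh, and exclude hidden enemies.
# Scoring favors closer goals, perception, non-participant enemies and, under an STW faction owner query, player-marked goals.
# A frustration test penalizes or temporarily ignores unreachable goals.
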EnvQuery /Irwin/AI/GoalSelection/Pawn/EQS_GoalSelection_Pawn_Wildlife_Terrestrial
options [{
  generator {
   name=EnvQueryGenerator_Composite_0
   class=EnvQueryGenerator_Composite
   properties {
    optionName=Perceived Enemies and Loots
    generators [
     &self:EnvQueryGenerator_Composite_0.FortQueryGenerator_Enemies_1
     &self:EnvQueryGenerator_Composite_0.FortQueryGenerator_LootGoalsAthena_0
    ]
   }
  }
  tests [
   FortQueryTest_GoalGameplayTags_0 {
    class=FortQueryTest_GoalGameplayTags
    properties {
     bDefineReferenceValue=false
     bRequireAllProvidedTagQueriesPass=false
     bScoreEncounterGoals=false
     bScoreEnemies=false
     bScoreSpecificAssignments=true
     bScoreWorldGoals=false
     bShouldLookupQueryByTag=true
     bShouldPassWhenQueryNotFound=true
     boolValue=true
     clampMaxType=None
     clampMinType=None
     filterType=Match
     floatValueMax=0
     floatValueMin=0
     multipleContextFilterOp=AllPass
     multipleContextScoreOp=AverageScore
     normalizationType=Absolute
     referenceValue=0
     scoreClampMax=0
     scoreClampMin=0
     scoringEquation=Constant
     scoringFactor=1
     testComment=Find Foraged Items
     testPurpose=Filter
     assignmentIds [{assignmentTeam=Monster, assignmentType=World, assignmentGameplayTags {gameplayTags [{tagName=NPC.Assignment.Loot}]}}]
     goalActorTagQuery {autoDescription="", tokenStreamVersion=0, userDescription="", queryTokenStream [], tagDictionary []}
     pawnOwnerTagQuery {autoDescription="", tokenStreamVersion=0, userDescription="", queryTokenStream [], tagDictionary []}
     prohibitedAssignmentIds []
     queryLookupTag {tagName=NPC.Assignment.Loot}
     tagQueryToMatch {autoDescription="", tokenStreamVersion=0, userDescription="", queryTokenStream [], tagDictionary []}
    }
   }
   FortQueryTest_GoalDistance_1 {
    class=FortQueryTest_GoalDistance
    properties {
     bScoreEncounterGoals=false
     bScoreEnemies=false
     bScoreSpecificAssignments=true
     bScoreWorldGoals=false
     clampMaxType=SpecifiedValue
     clampMinType=SpecifiedValue
     filterType=Maximum
     floatValueMax=3000
     scoreClampMax=5000
     scoreClampMin=200
     scoringEquation=InverseLinear
     scoringFactor=1
     testComment=Loot Scoring and Filtering
     testPurpose=FilterAndScore
     assignmentIds [{assignmentTeam=Monster, assignmentType=World, assignmentGameplayTags {gameplayTags [{tagName=NPC.Assignment.Loot}]}}]
    }
   }
   FortQueryTest_GoalDistance_0 {
    class=FortQueryTest_GoalDistance
    properties {
     bScoreEncounterGoals=true
     bScoreEnemies=true
     bScoreSpecificAssignments=false
     bScoreWorldGoals=false
     clampMaxType=SpecifiedValue
     clampMinType=SpecifiedValue
     filterType=Range
     floatValueMax=0
     scoreClampMax=5000
     scoreClampMin=200
     scoringEquation=InverseLinear
     scoringFactor=0.5
     testComment=Prefer closer things.
     testPurpose=Score
     assignmentIds []
    }
   }
   FortQueryTest_PerceptionAll_0 {
    class=FortQueryTest_PerceptionAll
    properties {
     clampMaxType=SpecifiedValue
     clampMinType=SpecifiedValue
     maxSenseAge=21
     minSenseAge=20
     scoreClampMax=1
     scoringFactor=1
     additionalSenseScores {&Class'&AthenaAISense_Hearing'=0.1}
     senseScores [0,0.1,1,0,0,0]
    }
   }
   FortQueryTest_GoalProject_0 {
    class=FortQueryTest_GoalProject
    properties {
     bDefineReferenceValue=false
     bScoreEncounterGoals=false
     bScoreEnemies=false
     bScoreSpecificAssignments=false
     bScoreWorldGoals=true
     boolValue=true
     clampMaxType=None
     clampMinType=None
     filterType=Match
     floatValueMax=0
     floatValueMin=0
     multipleContextFilterOp=AllPass
     multipleContextScoreOp=AverageScore
     normalizationType=Absolute
     referenceValue=0
     scoreClampMax=0
     scoreClampMin=0
     scoringEquation=Constant
     scoringFactor=1
     testComment="Only World (Loot) Goals near Nav"
     testPurpose=Filter
     assignmentIds []
     goalActorTagQuery {autoDescription="", tokenStreamVersion=0, userDescription="", queryTokenStream [], tagDictionary []}
     prohibitedAssignmentIds []
     projectionData {
      bCanDisableTrace=false
      bCanProjectDown=true
      bCanTraceOnGeometry=false
      bCanTraceOnNavMesh=true
      bOnlyBlockingHits=true
      bTraceComplex=false
      extentX=0
      extentY=10
      extentZ=10
      navigationFilter=null
      postProjectionVerticalOffset=0
      projectDown=256
      projectUp=256
      serializedChannel=ECC_Visibility
      traceChannel=TraceTypeQuery1
      traceMode=Navigation
      traceProfileName=None
      traceShape=Line
      versionNum=1
     }
    }
   }
   FortQueryTest_GoalFrustrationDiscouragement_0 {
    class=FortQueryTest_GoalFrustrationDiscouragement
    properties {
     bDefineReferenceValue=false
     bScoreEncounterGoals=true
     bScoreEnemies=true
     bScoreSpecificAssignments=false
     bScoreWorldGoals=true
     boolValue=true
     clampMaxType=SpecifiedValue
     clampMinType=SpecifiedValue
     filterType=Maximum
     floatValueMax=1
     floatValueMin=0
     multipleContextFilterOp=AllPass
     multipleContextScoreOp=AverageScore
     normalizationType=Absolute
     referenceValue=0
     scoreClampMax=1
     scoreClampMin=0
     scoringEquation=Constant
     scoringFactor=-1
     testComment="When a Goal can't be reached, we can choose to Frustrate on (IGNORE) that Goal for a period of time."
     testPurpose=FilterAndScore
     assignmentIds []
     goalActorTagQuery {autoDescription="", tokenStreamVersion=0, userDescription="", queryTokenStream [], tagDictionary []}
     prohibitedAssignmentIds []
    }
   }
   FortQueryTest_GoalGameplayTags_1 {
    class=FortQueryTest_GoalGameplayTags
    properties {
     bDefineReferenceValue=false
     bRequireAllProvidedTagQueriesPass=false
     bScoreEncounterGoals=false
     bScoreEnemies=true
     bScoreSpecificAssignments=false
     bScoreWorldGoals=false
     bShouldLookupQueryByTag=false
     bShouldPassWhenQueryNotFound=true
     boolValue=false
     clampMaxType=None
     clampMinType=None
     filterType=Match
     floatValueMax=0
     floatValueMin=0
     multipleContextFilterOp=AllPass
     multipleContextScoreOp=Multiply
     normalizationType=Absolute
     referenceValue=0
     scoreClampMax=0
     scoreClampMin=0
     scoringEquation=Constant
     scoringFactor=0.125
     testComment=Prefer non-participant
     testPurpose=Score
     assignmentIds []
     goalActorTagQuery {autoDescription="", tokenStreamVersion=0, userDescription="", queryTokenStream [], tagDictionary []}
     pawnOwnerTagQuery {autoDescription="", tokenStreamVersion=0, userDescription="", queryTokenStream [], tagDictionary []}
     prohibitedAssignmentIds []
     queryLookupTag {tagName=None}
     tagQueryToMatch {
      autoDescription=" ALL( Athena.Faction.GameParticipant )"
      tokenStreamVersion=0
      userDescription=""
      queryTokenStream [0,1,2,1,0]
      tagDictionary [{tagName=Athena.Faction.GameParticipant}]
     }
    }
   }
   FortQueryTest_GoalGameplayTags_2 {
    class=FortQueryTest_GoalGameplayTags
    properties {
     bDefineReferenceValue=false
     bRequireAllProvidedTagQueriesPass=false
     bScoreEncounterGoals=false
     bScoreEnemies=true
     bScoreSpecificAssignments=false
     bScoreWorldGoals=false
     bShouldLookupQueryByTag=false
     bShouldPassWhenQueryNotFound=true
     boolValue=false
     clampMaxType=None
     clampMinType=None
     filterType=Match
     floatValueMax=0
     floatValueMin=0
     multipleContextFilterOp=AllPass
     multipleContextScoreOp=Multiply
     normalizationType=Absolute
     referenceValue=0
     scoreClampMax=0
     scoreClampMin=0
     scoringEquation=Constant
     scoringFactor=0.125
     testComment=Exclude invisible targets
     testPurpose=Filter
     assignmentIds []
     goalActorTagQuery {autoDescription="", tokenStreamVersion=0, userDescription="", queryTokenStream [], tagDictionary []}
     pawnOwnerTagQuery {autoDescription="", tokenStreamVersion=0, userDescription="", queryTokenStream [], tagDictionary []}
     prohibitedAssignmentIds []
     queryLookupTag {tagName=None}
     tagQueryToMatch {
      autoDescription=" ALL( Athena.AI.Hidden )"
      tokenStreamVersion=0
      userDescription=""
      queryTokenStream [0,1,2,1,0]
      tagDictionary [{tagName=Athena.AI.Hidden}]
     }
    }
   }
   FortQueryTest_GoalMarkedByPlayer_0 {
    class=FortQueryTest_GoalMarkedByPlayer
    properties {
     bDefineReferenceValue=false
     bScoreEncounterGoals=true
     bScoreEnemies=true
     bScoreSpecificAssignments=false
     bScoreWorldGoals=true
     boolValue=true
     clampMaxType=None
     clampMinType=None
     filterType=Match
     floatValueMax=0
     floatValueMin=0
     multipleContextFilterOp=AllPass
     multipleContextScoreOp=AverageScore
     normalizationType=Absolute
     onlyConverterMarkedTargets=false
     referenceValue=0
     scoreClampMax=0
     scoreClampMin=0
     scoringEquation=Constant
     scoringFactor=5
     testComment=""
     testPurpose=Score
     assignmentIds []
     goalActorTagQuery {autoDescription="", tokenStreamVersion=0, userDescription="", queryTokenStream [], tagDictionary []}
     ownerTagQuery {
      autoDescription=" ANY( Athena.Faction.STW )"
      tokenStreamVersion=0
      userDescription=""
      queryTokenStream [0,1,1,1,0]
      tagDictionary [{tagName=Athena.Faction.STW}]
     }
     prohibitedAssignmentIds []
    }
   }
  ]
 }]
